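; Gameplay ability for the Suppressor barrel weapon mod, currently a stub with no suppression effect in its graph.
; On activation it casts the avatar FortPlayerPawn's current weapon to B_Ranged_Generic, stores it in CurrentWeapon and, if valid, prints a debug message.
; On end it only checks that the stored weapon is valid, with nothing wired after, and the failed passive activation event is empty.
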(event Ability|EventFailedToActivatePassiveAbility
  (params (ActorInfo :type Gameplay%20Ability%20Actor%20Info%20Structure)))

(event Ability|EventActivateAbility
  (bind _asb_ranged_generic (Utilities|Casting|CastToB_Ranged_Generic (Weapon|GetCurrentWeapon (CallFunction|GetAvatarasFortPlayerPawn))))
  (Variables|Default|SetCurrentWeapon _asb_ranged_generic)
  (Utilities|IsValid (Variables|Default|GetCurrentWeapon)
    (:Is%20Valid
      (Development|PrintString "Suppressor Applied"))))

(event Ability|EventOnEndAbility
  (params (bWasCancelled :type Boolean))
  (Utilities|IsValid (Variables|Default|GetCurrentWeapon)))
